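# Niagara effect for the Sunlight player emote, layering a red mark, smoke and sweat drops as sprites.
# RedDecal bursts once per loop, fading in with time-of-day alpha that is zero at night.
# Smoke spawns along a loop-age curve, follows source movement, animates material parameters and tints differently in the front end.
# SweatDrops combines curved spawn rate with a late burst, using spawn-group-masked velocities, gravity and stretch.
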NiagaraSystem /BRCosmetics/Animation/Game/MainPlayer/Emotes/Sunlight/FX/NS_Emote_Sunlight_01
dependencies {
 usedDataInterfaces [&NiagaraDataInterfaceCurve, &NiagaraDataInterfaceVector2DCurve]
 usedDynamicInputs []
 usedModules [&SystemState]
}
emitters {
 RedDecal {
  emitterUpdateScript [
   &EmitterState()
   &SpawnBurst_Instantaneous(Spawn Group=1, Spawn Time=2.15, Use Spawn Probability=false, UseLoopCountLimit=false)
  ]
  particleSpawnScript [
   &InitializeParticle(
    Color=<rgba 3 0.271 0.25 1>
    Lifetime=4
    Mass Mode=::NewEnumerator3
    Position Offset=<0 5 5>
    Sprite Size Mode=::NewEnumerator3
    Uniform Sprite Size=45
    UsePositionOffset=true
    Write Material Random=true
   )
   &DynamicMaterialParameters(Write Parameter Index 0=false)
  ]
  particleUpdateScript [
   &ParticleState()
   &ScaleSpriteSize()
   &ScaleColor(
    Scale Alpha=&FloatFromCurve(
     FloatCurve=DataInterface() {propertyValues {Curve {keys [<at 0 0 RCIM_Cubic>,<at 0.9 1 RCIM_Cubic>,<at 1 0>]}}}
     Scale Curve=&FN_Select_Float_from_Time_Of_Day(Morning=0.2, Day=0.2, Evening=0.2, Night=0)
    )
    ScaleA=true
    ScaleRGB=false
   ) {module=ScaleColor001}
  ]
  rendererClasses [&NiagaraSpriteRendererProperties]
  renderers [{rendererClass=&NiagaraSpriteRendererProperties}]
 }
 Smoke {
  emitterUpdateScript [
   &EmitterState()
   &SpawnRate(
    SpawnRate=&FloatFromCurve(
     FloatCurve=DataInterface() {propertyValues {Curve {keys [
         {arriveTangent=1.6728, interpMode=RCIM_Cubic, leaveTangent=1.6728, tangentMode=RCTM_User, time=0.1, value=0}
         <at 0.7 1 RCIM_Cubic>
         {arriveTangent=-1.4163, tangentMode=RCTM_User, time=0.8, value=0}
        ]}}}
     CurveIndex=$Emitter.NormalizedLoopAge
     Scale Curve=13
    )
    Use Spawn Probability=false
   ) {module=SpawnRate001}
  ]
  particleSpawnScript [
   &InitializeParticle(
    Color=&LerpLinearColors(
     StartColor=<rgba 1 1 1 1>
     EndColor=<rgba 0.7266 0.7383 0.75 1>
     LerpFactor=&FN_FrontEndFloatModifier(Gameplay=0, Front End=1)
    )
    Lifetime Max=1
    Lifetime Min=0.7
    Lifetime Mode=::NewEnumerator1
    Sprite Rotation Mode=::NewEnumerator2
    Sprite Size Mode=::NewEnumerator1
    Uniform Sprite Size Max=40
    Uniform Sprite Size Min=30
    UsePositionOffset=false
    Write Material Random=true
   )
   &CameraOffset(Camera Offset Amount=11)
   &ShapeLocation(MaskSpawnGroup=false, Offset=<0 0 -10>, Offset Mode=::NewEnumerator0, Override Seed=false, Sphere Radius=12, UseOwnerScale=true) {module=ShapeLocation001}
   &AddVelocity(Velocity=&RandomRangeVector(Minimum=<20 0 0>, Maximum=<50 0 0>))
  ]
  particleUpdateScript [
   &InheritSourceMovement(ApplyForce=false, ApplyPosition=true, ApplyPositionFalloff=false)
   &SolveForcesAndVelocity(Clamp Velocity=false, Limit Acceleration=false)
   &ParticleState()
   &ScaleSpriteSize(Uniform Curve Sprite Scale=DataInterface() {propertyValues {Curve {keys [<at 0 0.6 RCIM_Cubic>, {arriveTangent=-0.15, interpMode=RCIM_Cubic, leaveTangent=-0.15, time=0.6, value=1}, <at 1 0.45 RCIM_Cubic>]}}})
   &SpriteRotationRate(Rotation Rate=&RandomRangeFloat(Minimum=-0.1, Maximum=0.1))
   &DynamicMaterialParameters(
    Index 0 Param 1=&Add_Float(
     A=&RandomRangeFloat(Minimum=0, Maximum=1)
     B=&FloatFromCurve(
      FloatCurve=DataInterface() {propertyValues {Curve {keys [
          {arriveTangent=2.7533, interpMode=RCIM_Cubic, leaveTangent=1.8958, tangentMode=RCTM_User, time=0, value=1}
          {arriveTangent=0.8292, interpMode=RCIM_Cubic, leaveTangent=0.8292, tangentMode=RCTM_User, time=0.321, value=1.3453}
          {arriveTangent=0.0529, interpMode=RCIM_Cubic, leaveTangent=0.2021, tangentMode=RCTM_User, time=1, value=1.5}
         ]}}}
      CurveIndex=&Add_Float(A=&Multiply_Float(A=$Particles.Initial.MaterialRandom, B=1), B=$Particles.NormalizedAge)
     )
    )
    Index 0 Param 2=&FloatFromCurve(FloatCurve=DataInterface() {propertyValues {Curve {keys [
         {arriveTangent=0.0453, interpMode=RCIM_Cubic, tangentMode=RCTM_Break, time=0, value=-1}
         {arriveTangent=0.0604, interpMode=RCIM_Cubic, leaveTangent=0.0604, tangentMode=RCTM_User, time=1, value=-1}
        ]}}})
    Index 0 Param 3=&FloatFromCurve(FloatCurve=DataInterface() {propertyValues {Curve {keys [
         {arriveTangent=-7.8561, time=0, value=0}
         {arriveTangent=-1.6283, interpMode=RCIM_Cubic, leaveTangent=0.2723, tangentMode=RCTM_Break, time=0.18, value=0}
        ]}}})
    Index 0 Param 4=&FloatFromCurve(FloatCurve=DataInterface() {propertyValues {Curve {keys [{interpMode=RCIM_Cubic, leaveTangent=-1.5927, tangentMode=RCTM_Break, time=0, value=1}, <at 1 0>]}}})
    Index 1 Param 2=&RandomRangeFloat(Minimum=0, Maximum=1)
    Index 1 Param 4=&FloatFromCurve(
     FloatCurve=DataInterface() {propertyValues {Curve {keys [{arriveTangent=-0.3224, interpMode=RCIM_Cubic, leaveTangent=-0.3224, tangentMode=RCTM_User, time=0, value=1}, <at 1 0.9>]}}}
     Scale Curve=&RandomRangeFloat(Minimum=1.1, Maximum=1.4)
    )
    Index 2 Param 1=&FloatFromCurve(FloatCurve=DataInterface() {propertyValues {Curve {keys [
         {interpMode=RCIM_Cubic, leaveTangent=4.2485, tangentMode=RCTM_Break, time=0, value=0}
         {arriveTangent=1.0548, interpMode=RCIM_Cubic, leaveTangent=1.0548, tangentMode=RCTM_User, time=0.17, value=0.5715}
         <at 1 1>
        ]}}})
    Index 2 Param 2=&FloatFromCurve(
     FloatCurve=DataInterface() {propertyValues {Curve {keys [
         {arriveTangent=3.069, interpMode=RCIM_Cubic, leaveTangent=3.069, tangentMode=RCTM_User, time=0, value=0.2}
         {arriveTangent=0.0259, interpMode=RCIM_Cubic, leaveTangent=0.0259, tangentMode=RCTM_User, time=1, value=1.4}
        ]}}}
     Scale Curve=&RandomRangeFloat(Minimum=0.2, Maximum=1.25)
    )
    Index 2 Param 3=&FloatFromCurve(FloatCurve=DataInterface() {propertyValues {Curve {keys [
         {interpMode=RCIM_Cubic, leaveTangent=2.8489, tangentMode=RCTM_Break, time=0, value=1}
         {interpMode=RCIM_Cubic, leaveTangent=11.1233, tangentMode=RCTM_Break, time=1, value=2}
        ]}}})
    Index 2 Param 4=&FloatFromCurve(FloatCurve=DataInterface() {propertyValues {Curve {keys [
         {arriveTangent=-0.6867, interpMode=RCIM_Cubic, leaveTangent=-0.6867, tangentMode=RCTM_User, time=0, value=1}
         {interpMode=RCIM_Cubic, leaveTangent=-1.2956, tangentMode=RCTM_Break, time=1, value=0.5}
        ]}}})
    Module.Param0WriteEnabled=true
    Module.Param0WriteEnabled1=true
    Module.Param0WriteEnabled2=true
    Module.Param1WriteEnabled=false
    Module.Param1WriteEnabled1=true
    Module.Param1WriteEnabled2=true
    Module.Param2WriteEnabled=false
    Module.Param2WriteEnabled1=true
    Module.Param2WriteEnabled2=true
    Module.Param3WriteEnabled=true
    Module.Param3WriteEnabled1=false
    Module.Param3WriteEnabled2=true
    Write Parameter Index 1=true
    Write Parameter Index 2=true
   ) {module=DynamicMaterialParameters001}
   &ScaleColor(
    Scale Alpha=&FloatFromCurve(FloatCurve=DataInterface() {propertyValues {Curve {keys [<at 0 0 RCIM_Cubic>,<at 0.3 0.9 RCIM_Cubic>,<at 1 0>]}}}, Scale Curve=0.06)
    ScaleA=true
    ScaleRGB=true
   )
  ]
  rendererClasses [&NiagaraSpriteRendererProperties]
  renderers [
   {rendererClass=&NiagaraSpriteRendererProperties}
   {rendererClass=&NiagaraSpriteRendererProperties, rendererIndex=1}
  ]
 }
 SweatDrops {
  emitterUpdateScript [
   &EmitterState()
   &SpawnBurst_Instantaneous(Spawn Count=8, Spawn Group=1, Spawn Time=5.7, Use Spawn Probability=false, UseLoopCountLimit=false)
   &SpawnRate(
    SpawnRate=&FloatFromCurve(
     FloatCurve=DataInterface() {propertyValues {Curve {keys [
         {arriveTangent=2.1826, interpMode=RCIM_Cubic, leaveTangent=2.1826, tangentMode=RCTM_User, time=0, value=0}
         <at 0.7 1 RCIM_Cubic>
         {arriveTangent=-1.4163, tangentMode=RCTM_User, time=0.9, value=0}
        ]}}}
     CurveIndex=$Emitter.NormalizedLoopAge
     Scale Curve=9
    )
    Use Spawn Probability=false
   ) {module=SpawnRate001}
  ]
  particleSpawnScript [
   &InitializeParticle(
    Lifetime Max=0.4
    Lifetime Min=0.25
    Lifetime Mode=::NewEnumerator1
    Mass Mode=::NewEnumerator3
    Sprite Size Max=<1.25 4>
    Sprite Size Min=<0.8 3>
    Sprite Size Mode=::NewEnumerator2
    UsePositionOffset=false
    Write Material Random=true
   )
   &AddVelocity(Velocity=&MaskVectorBySpawnGroup(
     Masked Vector=&RandomRangeVector(Minimum=<10 10 -40>, Maximum=<30 80 40>)
     Passthrough Vector=&RandomRangeVector(Minimum=<100 40 60>, Maximum=<200 50 170>)
    )) {module=AddVelocity002}
   &ApplyInitialForces()
   &ShapeLocation(
    MaskSpawnGroup=false
    Offset=<7 6 0>
    Offset Coordinate Space=::Local
    Offset Mode=::NewEnumerator0
    Override Seed=false
    Sphere Radius=7
    UseOwnerScale=true
   ) {module=ShapeLocation002}
  ]
  particleUpdateScript [
   &ParticleState()
   &GravityForce(Gravity=<0 0 -700>)
   &ScaleSpriteSize(Scale Factor=&Vector2DFromCurve(
     Vector2Curve=DataInterface() {propertyValues {
       XCurve {keys [<at 0.25 1 RCIM_Cubic>, {interpMode=RCIM_Cubic, leaveTangent=-2.9373, time=1, value=0.3}]}
       YCurve {keys [
         {arriveTangent=-1.4216, interpMode=RCIM_Cubic, time=0.0067, value=0.3562}
         {arriveTangent=2.9367, interpMode=RCIM_Cubic, leaveTangent=2.9367, tangentMode=RCTM_User, time=0.2953, value=1.3705}
         {interpMode=RCIM_Cubic, leaveTangent=7.4188, time=1, value=2}
        ]}
      }}
     ScaleCurve=<1.3 1.3>
    ))
   &Drag(Drag=0.5, Use Linear Drag=true, Use Rotational Drag=false)
   &DynamicMaterialParameters(
    Index 0 Param 1=0.8
    Index 0 Param 2=&FloatFromCurve(FloatCurve=DataInterface() {propertyValues {Curve {keys [{arriveTangent=-1.8884, interpMode=RCIM_Cubic, leaveTangent=-1.8884, tangentMode=RCTM_User, time=0, value=1}, <at 1 0.26>]}}})
    Index 0 Param 3=1
    Index 0 Param 4=1
    Module.Param0WriteEnabled=true
    Module.Param1WriteEnabled=true
    Module.Param2WriteEnabled=false
    Module.Param3WriteEnabled=false
   )
   &ScaleColor(
    Scale Alpha=&FloatFromCurve(
     FloatCurve=DataInterface() {propertyValues {Curve {keys [<at 0 1>, {arriveTangent=-1, interpMode=RCIM_Cubic, leaveTangent=-1, time=0.8, value=1}, <at 1 0>]}}}
     Scale Curve=&RandomRangeFloat(Minimum=0.1, Maximum=0.2)
    )
    ScaleA=true
    ScaleRGB=false
   ) {module=ScaleColor001}
   &InheritSourceMovement(ApplyForce=false, ApplyPosition=true, ApplyPositionFalloff=false)
   &SolveForcesAndVelocity(Clamp Velocity=false, Limit Acceleration=false)
  ]
  rendererClasses [&NiagaraSpriteRendererProperties]
  renderers [{rendererClass=&NiagaraSpriteRendererProperties}]
 }
}
summary {
 systemName=NS_Emote_Sunlight_01
 emitters [
  {emitterName=SweatDrops, rendererClasses [&NiagaraSpriteRendererProperties]}
  {emitterName=Smoke, rendererClasses [&NiagaraSpriteRendererProperties]}
  {emitterName=RedDecal, rendererClasses [&NiagaraSpriteRendererProperties]}
 ]
 userVariables [
  User.bIsFrontEnd {type=NiagaraBool}
  User.bIsFrontEndPreview {type=NiagaraBool}
 ]
}
