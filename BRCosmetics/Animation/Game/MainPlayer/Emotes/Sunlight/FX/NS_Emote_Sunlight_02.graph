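# Niagara effect for the Sunlight emote, building a glowing flare with lens-flare elements, a light and drifting light rays.
# Headlights burst-spawns one stack of sprite and light renderers, driven by emitter-level size, direction and faded-in color variables.
# LightRays spawns randomized SubUV sprites in a box that grow, drift and fade with the emitter loop.
# Time-of-day selectors switch warm colors to blue at night, and front-end modifiers tune some values.
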NiagaraSystem /BRCosmetics/Animation/Game/MainPlayer/Emotes/Sunlight/FX/NS_Emote_Sunlight_02
dependencies {
 usedDataInterfaces [&NiagaraDataInterfaceColorCurve, &NiagaraDataInterfaceCurve]
 usedDynamicInputs []
 usedModules [&SystemState]
}
emitters {
 Headlights {
  emitterSpawnScript [/SetVariables(
    Emitter.CapsuleGlowLength=44
    Emitter.CapsuleGlowSize:Vector2f=&Vector2DFromFloat(Value=22)
    Emitter.CapsuleGlowStemLength=33
    Emitter.CapsuleGlowStemSize:Vector2f=&Vector2DFromFloat(Value=16)
    Emitter.FlareCameraOffset=22
    Emitter.FlareSize:Vector2f=&Vector2DFromFloat(Value=133)
    Emitter.GlowCameraOffset=-12
    Emitter.GlowSize:Vector2f=&Vector2DFromFloat(Value=100)
    Emitter.LensFlareElement1Offset:NiagaraFloat=&FN_FrontEndFloatModifier(Gameplay=0.333, Front End=0.333)
    Emitter.LensFlareElement1Size:Vector2f=&Vector2DFromFloat(Value=&FN_FrontEndFloatModifier(Gameplay=0.133, Front End=0.133))
    Emitter.LensFlareElement2Offset:NiagaraFloat=&FN_FrontEndFloatModifier(Gameplay=0.533, Front End=0.533)
    Emitter.LensFlareElement2Size:Vector2f=&Vector2DFromFloat(Value=&FN_FrontEndFloatModifier(Gameplay=0.2, Front End=0.2))
    Emitter.LensFlareElement3Offset:NiagaraFloat=&FN_FrontEndFloatModifier(Gameplay=0.633, Front End=0.633)
    Emitter.LensFlareElement3Size:Vector2f=&Vector2DFromFloat(Value=&FN_FrontEndFloatModifier(Gameplay=0.233, Front End=0.233))
    Emitter.LensFlareRainbowOffset:NiagaraFloat=&FN_FrontEndFloatModifier(Gameplay=0.1, Front End=0.1)
    Emitter.LensFlareRainbowSize:Vector2f=&Vector2DFromFloat(Value=&FN_FrontEndFloatModifier(Gameplay=0.333, Front End=0.333))
   ) {isSetParameters=true}]
  emitterUpdateScript [
   &EmitterState(
    Inactive Response=::NewEnumerator1
    Life Cycle Mode=::NewEnumerator1
    Loop Behavior=::NewEnumerator1
    Loop Duration Mode=::NewEnumerator1
    UseLoopDelay=false
   )
   &SpawnBurst_Instantaneous(Use Spawn Probability=false, UseLoopCountLimit=false)
   /SetVariables(
    Emitter.Color01:LinearColor=&FN_Select_Color_from_Time_Of_Day(
     Morning=&ColorFromCurve(
      ColorCurve=DataInterface() {propertyValues {
        AlphaCurve {keys [<at 0.1481 0 RCIM_Cubic>,<at 0.3417 0.3 RCIM_Cubic>]}
        BlueCurve {keys [<at 0.0857 0.2169 RCIM_Cubic>]}
        GreenCurve {keys [<at 0.0857 0.5903 RCIM_Cubic>]}
        RedCurve {keys [<at 0.0857 1 RCIM_Cubic>]}
       }}
      CurveIndex=$Emitter.Age
     )
     Day=&ColorFromCurve(
      ColorCurve=DataInterface() {propertyValues {
        AlphaCurve {keys [<at 0.1481 0 RCIM_Cubic>,<at 0.3417 0.3 RCIM_Cubic>]}
        BlueCurve {keys [<at 0.0857 0.2169 RCIM_Cubic>]}
        GreenCurve {keys [<at 0.0857 0.5903 RCIM_Cubic>]}
        RedCurve {keys [<at 0.0857 1 RCIM_Cubic>]}
       }}
      CurveIndex=$Emitter.Age
     )
     Evening=&ColorFromCurve(
      ColorCurve=DataInterface() {propertyValues {
        AlphaCurve {keys [<at 0.1481 0 RCIM_Cubic>,<at 0.3417 0.3 RCIM_Cubic>]}
        BlueCurve {keys [<at 0.0857 0.2169 RCIM_Cubic>]}
        GreenCurve {keys [<at 0.0857 0.5903 RCIM_Cubic>]}
        RedCurve {keys [<at 0.0857 1 RCIM_Cubic>]}
       }}
      CurveIndex=$Emitter.Age
     )
     Night=&ColorFromCurve(
      ColorCurve=DataInterface() {propertyValues {
        AlphaCurve {keys [<at 0.1481 0 RCIM_Cubic>,<at 0.3417 0.3 RCIM_Cubic>]}
        BlueCurve {keys [<at 0.0857 1 RCIM_Cubic>]}
        GreenCurve {keys [<at 0.0857 0.8206 RCIM_Cubic>]}
        RedCurve {keys [<at 0.0857 0.3594 RCIM_Cubic>]}
       }}
      CurveIndex=$Emitter.Age
     )
    )
    Emitter.Color02:LinearColor=&FN_Select_Color_from_Time_Of_Day(
     Morning=&ColorFromCurve(
      ColorCurve=DataInterface() {propertyValues {
        AlphaCurve {keys [<at 0.1597 0 RCIM_Cubic>,<at 0.3533 0.4 RCIM_Cubic>]}
        BlueCurve {keys [<at 0.0857 0 RCIM_Cubic>]}
        GreenCurve {keys [<at 0.0857 0.4059 RCIM_Cubic>]}
        RedCurve {keys [<at 0.0857 1 RCIM_Cubic>]}
       }}
      CurveIndex=$Emitter.Age
     )
     Day=&ColorFromCurve(
      ColorCurve=DataInterface() {propertyValues {
        AlphaCurve {keys [<at 0.1597 0 RCIM_Cubic>,<at 0.3533 0.4 RCIM_Cubic>]}
        BlueCurve {keys [<at 0.0857 0 RCIM_Cubic>]}
        GreenCurve {keys [<at 0.0857 0.4059 RCIM_Cubic>]}
        RedCurve {keys [<at 0.0857 1 RCIM_Cubic>]}
       }}
      CurveIndex=$Emitter.Age
     )
     Evening=&ColorFromCurve(
      ColorCurve=DataInterface() {propertyValues {
        AlphaCurve {keys [<at 0.1597 0 RCIM_Cubic>,<at 0.3533 0.4 RCIM_Cubic>]}
        BlueCurve {keys [<at 0.0857 0 RCIM_Cubic>]}
        GreenCurve {keys [<at 0.0857 0.4059 RCIM_Cubic>]}
        RedCurve {keys [<at 0.0857 1 RCIM_Cubic>]}
       }}
      CurveIndex=$Emitter.Age
     )
     Night=&ColorFromCurve(
      ColorCurve=DataInterface() {propertyValues {
        AlphaCurve {keys [<at 0.1597 0 RCIM_Cubic>,<at 0.3533 0.4 RCIM_Cubic>]}
        BlueCurve {keys [<at 0.0857 1 RCIM_Cubic>]}
        GreenCurve {keys [<at 0.0857 0.8998 RCIM_Cubic>]}
        RedCurve {keys [<at 0.0857 0.4995 RCIM_Cubic>]}
       }}
      CurveIndex=$Emitter.Age
     )
    )
    Emitter.Color03:LinearColor=&FN_Select_Color_from_Time_Of_Day(
     Morning=&ColorFromCurve(
      ColorCurve=DataInterface() {propertyValues {
        AlphaCurve {keys [<at 0.1646 0 RCIM_Cubic>,<at 0.3681 0.5 RCIM_Cubic>]}
        BlueCurve {keys [<at 0.0857 0.0478 RCIM_Cubic>]}
        GreenCurve {keys [<at 0.0857 0.3377 RCIM_Cubic>]}
        RedCurve {keys [<at 0.0857 0.5732 RCIM_Cubic>]}
       }}
      CurveIndex=$Emitter.Age
     )
     Day=&ColorFromCurve(
      ColorCurve=DataInterface() {propertyValues {
        AlphaCurve {keys [<at 0.1646 0 RCIM_Cubic>,<at 0.3681 0.5 RCIM_Cubic>]}
        BlueCurve {keys [<at 0.0857 0.0478 RCIM_Cubic>]}
        GreenCurve {keys [<at 0.0857 0.3377 RCIM_Cubic>]}
        RedCurve {keys [<at 0.0857 0.5732 RCIM_Cubic>]}
       }}
      CurveIndex=$Emitter.Age
     )
     Evening=&ColorFromCurve(
      ColorCurve=DataInterface() {propertyValues {
        AlphaCurve {keys [<at 0.1646 0 RCIM_Cubic>,<at 0.3681 0.5 RCIM_Cubic>]}
        BlueCurve {keys [<at 0.0857 0.0478 RCIM_Cubic>]}
        GreenCurve {keys [<at 0.0857 0.3377 RCIM_Cubic>]}
        RedCurve {keys [<at 0.0857 0.5732 RCIM_Cubic>]}
       }}
      CurveIndex=$Emitter.Age
     )
     Night=&ColorFromCurve(
      ColorCurve=DataInterface() {propertyValues {
        AlphaCurve {keys [<at 0.1646 0 RCIM_Cubic>,<at 0.3681 0.5 RCIM_Cubic>]}
        BlueCurve {keys [<at 0.0857 0.5732 RCIM_Cubic>]}
        GreenCurve {keys [<at 0.0857 0.4624 RCIM_Cubic>]}
        RedCurve {keys [<at 0.0857 0.1493 RCIM_Cubic>]}
       }}
      CurveIndex=$Emitter.Age
     )
    )
    Emitter.Color04:LinearColor=&FN_Select_Color_from_Time_Of_Day(
     Morning=&ColorFromCurve(
      ColorCurve=DataInterface() {propertyValues {
        AlphaCurve {keys [<at 0.1944 0 RCIM_Cubic>,<at 0.3863 0.14 RCIM_Cubic>]}
        BlueCurve {keys [<at 0.0857 0.1958 RCIM_Cubic>]}
        GreenCurve {keys [<at 0.0857 0.686 RCIM_Cubic>]}
        RedCurve {keys [<at 0.0857 1 RCIM_Cubic>]}
       }}
      CurveIndex=$Emitter.Age
     )
     Day=&ColorFromCurve(
      ColorCurve=DataInterface() {propertyValues {
        AlphaCurve {keys [<at 0.1944 0 RCIM_Cubic>,<at 0.3863 0.14 RCIM_Cubic>]}
        BlueCurve {keys [<at 0.0857 0.1958 RCIM_Cubic>]}
        GreenCurve {keys [<at 0.0857 0.686 RCIM_Cubic>]}
        RedCurve {keys [<at 0.0857 1 RCIM_Cubic>]}
       }}
      CurveIndex=$Emitter.Age
     )
     Evening=&ColorFromCurve(
      ColorCurve=DataInterface() {propertyValues {
        AlphaCurve {keys [<at 0.1944 0 RCIM_Cubic>,<at 0.3863 0.14 RCIM_Cubic>]}
        BlueCurve {keys [<at 0.0857 0.1958 RCIM_Cubic>]}
        GreenCurve {keys [<at 0.0857 0.686 RCIM_Cubic>]}
        RedCurve {keys [<at 0.0857 1 RCIM_Cubic>]}
       }}
      CurveIndex=$Emitter.Age
     )
     Night=&ColorFromCurve(
      ColorCurve=DataInterface() {propertyValues {
        AlphaCurve {keys [<at 0.1944 0 RCIM_Cubic>,<at 0.3863 0.14 RCIM_Cubic>]}
        BlueCurve {keys [<at 0.0857 1 RCIM_Cubic>]}
        GreenCurve {keys [<at 0.0857 0.8997 RCIM_Cubic>]}
        RedCurve {keys [<at 0.0857 0.5401 RCIM_Cubic>]}
       }}
      CurveIndex=$Emitter.Age
     )
    )
    Emitter.Color05:LinearColor=&FN_Select_Color_from_Time_Of_Day(
     Morning=&ColorFromCurve(
      ColorCurve=DataInterface() {propertyValues {
        AlphaCurve {keys [<at 0.198 0 RCIM_Cubic>,<at 0.3836 1 RCIM_Cubic>]}
        BlueCurve {keys [<at 0.0644 0 RCIM_Cubic>,<at 1 1 RCIM_Cubic>]}
        GreenCurve {keys [<at 0.0644 15.5627 RCIM_Cubic>,<at 1 1 RCIM_Cubic>]}
        RedCurve {keys [<at 0.0644 40 RCIM_Cubic>,<at 1 1 RCIM_Cubic>]}
       }}
      CurveIndex=$Emitter.NormalizedLoopAge
     )
     Day=&ColorFromCurve(
      ColorCurve=DataInterface() {propertyValues {
        AlphaCurve {keys [<at 0.198 0 RCIM_Cubic>,<at 0.3836 1 RCIM_Cubic>]}
        BlueCurve {keys [<at 0.0644 0 RCIM_Cubic>,<at 1 1 RCIM_Cubic>]}
        GreenCurve {keys [<at 0.0644 15.5627 RCIM_Cubic>,<at 1 1 RCIM_Cubic>]}
        RedCurve {keys [<at 0.0644 40 RCIM_Cubic>,<at 1 1 RCIM_Cubic>]}
       }}
      CurveIndex=$Emitter.NormalizedLoopAge
     )
     Evening=&ColorFromCurve(
      ColorCurve=DataInterface() {propertyValues {
        AlphaCurve {keys [<at 0.198 0 RCIM_Cubic>,<at 0.3836 1 RCIM_Cubic>]}
        BlueCurve {keys [<at 0.0644 0 RCIM_Cubic>,<at 1 1 RCIM_Cubic>]}
        GreenCurve {keys [<at 0.0644 15.5627 RCIM_Cubic>,<at 1 1 RCIM_Cubic>]}
        RedCurve {keys [<at 0.0644 40 RCIM_Cubic>,<at 1 1 RCIM_Cubic>]}
       }}
      CurveIndex=$Emitter.NormalizedLoopAge
     )
     Night=&ColorFromCurve(
      ColorCurve=DataInterface() {propertyValues {
        AlphaCurve {keys [<at 0.198 0 RCIM_Cubic>,<at 0.3836 1 RCIM_Cubic>]}
        BlueCurve {keys [<at 0.0644 40 RCIM_Cubic>,<at 1 1 RCIM_Cubic>]}
        GreenCurve {keys [<at 0.0644 32.5191 RCIM_Cubic>,<at 1 1 RCIM_Cubic>]}
        RedCurve {keys [<at 0.0644 9.7917 RCIM_Cubic>,<at 1 1 RCIM_Cubic>]}
       }}
      CurveIndex=$Emitter.NormalizedLoopAge
     )
    )
   ) {isSetParameters=true}
   /SetVariables(
    Emitter.CapsuleGlowAxis:Vector3f=&TransformVector(Vector=&RotateVectorByQuaternion(Vector To Rotate=<0 0 1>, Quaternion=<0 0 0 1>), Destination Space=::World)
    Emitter.HeadlightColor:LinearColor=&MultiplyLinearColor(Linear Color=<rgba 3 2.0699 0 1>, Scale Factor=&MakeVector4_FromFloat(Float=1))
    Emitter.HeadlightDirection:Vector3f=&RotateVectorByQuaternion(Vector To Rotate=<0 1 0>, Quaternion=<0 0 0 1>)
   ) {isSetParameters=true}
   /SetVariables(
    Emitter.BeamAlignment:Vector3f=&Multiply_VectorByFloat(Vector=$Emitter.HeadlightDirection, Float=0.5314)
    Emitter.BeamColor:LinearColor=&MultiplyLinearColor(Linear Color=$Emitter.HeadlightColor, Scale Factor=&MakeVector4_FromFloat(Float=0.333))
    Emitter.CapsuleGlowColor:LinearColor=&MultiplyLinearColor(Linear Color=$Emitter.HeadlightColor, Scale Factor=&MakeVector4_FromFloat(Float=1))
    Emitter.CapsuleGlowDynamic:Vector4f=&MakeFloat4FromVectorAndFloat("Vector (RGB)"=$Emitter.CapsuleGlowAxis, "Float (Alpha)"=$Emitter.CapsuleGlowLength)
    Emitter.CapsuleGlowStemColor:LinearColor=&MultiplyLinearColor(Linear Color=$Emitter.HeadlightColor, Scale Factor=&MakeVector4_FromFloat(Float=0.666))
    Emitter.CapsuleGlowStemDynamic:Vector4f=&MakeFloat4FromVectorAndFloat("Vector (RGB)"=$Emitter.CapsuleGlowAxis, "Float (Alpha)"=$Emitter.CapsuleGlowStemLength)
    Emitter.FlareColor:LinearColor=&MultiplyLinearColor(Linear Color=$Emitter.HeadlightColor, Scale Factor=&MakeVector4_FromFloat(Float=1))
    Emitter.GlowColor:LinearColor=&MultiplyLinearColor(Linear Color=$Emitter.HeadlightColor, Scale Factor=&MakeVector4_FromFloat(Float=0.333))
    Emitter.LensFlareColor:LinearColor=&MultiplyLinearColor(Linear Color=$Emitter.HeadlightColor, Scale Factor=&MakeVector4_FromFloat(Float=0.4))
    Emitter.LensFlareElement1Dynamic:Vector4f=&MakeFloat4FromVectorAndFloat("Vector (RGB)"=$Emitter.HeadlightDirection, "Float (Alpha)"=$Emitter.Initial.LensFlareElement1Offset)
    Emitter.LensFlareElement2Dynamic:Vector4f=&MakeFloat4FromVectorAndFloat("Vector (RGB)"=$Emitter.HeadlightDirection, "Float (Alpha)"=$Emitter.Initial.LensFlareElement2Offset)
    Emitter.LensFlareElement3Dynamic:Vector4f=&MakeFloat4FromVectorAndFloat("Vector (RGB)"=$Emitter.HeadlightDirection, "Float (Alpha)"=$Emitter.Initial.LensFlareElement3Offset)
    Emitter.LensFlareRainbowDynamic:Vector4f=&MakeFloat4FromVectorAndFloat("Vector (RGB)"=$Emitter.HeadlightDirection, "Float (Alpha)"=$Emitter.Initial.LensFlareRainbowOffset)
    Emitter.LightColor:LinearColor=&MultiplyLinearColor(Linear Color=$Emitter.HeadlightColor, Scale Factor=&MakeVector4_FromFloat(Float=3))
   ) {isSetParameters=true}
   /SetVariables(Emitter.LightPos:NiagaraPosition=&AddVectorToPosition(Position=<0 0 0>, Vector=<-1 0 10>)) {isSetParameters=true}
   /SetVariables(Emitter.GlowPos:NiagaraPosition=&AddVectorToPosition(Position=<0 0 0>, Vector=<0 1 0>)) {isSetParameters=true}
  ]
  particleSpawnScript [&InitializeParticle(
    Color=<rgba 1 0.9417 0.3 1>
    Position Mode=::NewEnumerator0
    Sprite Size Mode=::NewEnumerator3
    Uniform Sprite Size=0.2
    Write Material Random=true
   )]
  particleUpdateScript [&ParticleState(Kill Particles When Lifetime Has Elapsed=false, Let Infinitely Lived Particles Die When Emitter Deactivates=true)]
  rendererClasses [&NiagaraSpriteRendererProperties, &NiagaraLightRendererProperties]
  renderers [
   {rendererClass=&NiagaraSpriteRendererProperties}
   {rendererClass=&NiagaraSpriteRendererProperties, rendererIndex=1}
   {rendererClass=&NiagaraSpriteRendererProperties, rendererIndex=2}
   {rendererClass=&NiagaraSpriteRendererProperties, rendererIndex=3}
   {rendererClass=&NiagaraSpriteRendererProperties, rendererIndex=4}
   {rendererClass=&NiagaraLightRendererProperties, rendererIndex=5}
  ]
 }
 LightRays {
  emitterUpdateScript [
   &EmitterState(Inactive Response=::NewEnumerator1, Life Cycle Mode=::NewEnumerator1, Loop Duration=14.2, UseLoopDelay=false)
   &SpawnBurst_Instantaneous(Use Spawn Probability=false, UseLoopCountLimit=false)
   &SpawnRate(SpawnRate=3, Use Spawn Probability=false)
  ]
  particleSpawnScript [
   &InitializeParticle(
    Color=&FN_Select_Color_from_Time_Of_Day(
     Morning=&RandomRangeLinearColor(Minimum=<rgba 1 0.5913 0.2825 1>, Maximum=<rgba 1 0.7278 0.3352 1>)
     Day=&RandomRangeLinearColor(Minimum=<rgba 1 0.5913 0.2825 1>, Maximum=<rgba 1 0.7278 0.3352 1>)
     Evening=&RandomRangeLinearColor(Minimum=<rgba 1 0.5913 0.2825 1>, Maximum=<rgba 1 0.7278 0.3352 1>)
     Night=&RandomRangeLinearColor(Minimum=<rgba 0.2865 0.9347 1 1>, Maximum=<rgba 0.3646 0.7053 1 1>)
    )
    Lifetime Min=1.5
    Lifetime Mode=::NewEnumerator1
    Sprite Rotation Angle=&RandomRangeFloat(Minimum=-30, Maximum=20)
    Sprite Rotation Mode=::NewEnumerator0
    Sprite Size Max=<90 50>
    Sprite Size Min=<80 40>
    Sprite Size Mode=::NewEnumerator2
    UsePositionOffset=true
    Write Material Random=true
   )
   &SubUVAnimation(End Frame=3, SubUV Animation Mode:ENiagara_SubUVLookupMode=::NewEnumerator2)
   &AddVelocity(Velocity=<0.439 1 0>)
   &ShapeLocation(
    Box Midpoint=<1.1 0.5 0.5>
    Box Size=<9 0 20>
    MaskSpawnGroup=false
    Override Seed=false
    Shape Primitive=::NewEnumerator2
    Surface Only=false
    UseOwnerScale=true
   )
   /SetVariables(Particles.SizeLowMed:Vector2f=&RandomRangeVector2D(Minimum=<90 180>, Maximum=<120 180>)) {isSetParameters=true}
  ]
  particleUpdateScript [
   &SolveForcesAndVelocity(Clamp Velocity=false, Limit Acceleration=false)
   &ParticleState()
   &ScaleSpriteSize(Scale Factor=&MakeVector2D(
     X=&FloatFromCurve(FloatCurve=DataInterface() {propertyValues {Curve {keys [
          {arriveTangent=0.8031, interpMode=RCIM_Cubic, leaveTangent=0.8031, tangentMode=RCTM_User, time=0, value=0.625}
          {arriveTangent=0.115, interpMode=RCIM_Cubic, leaveTangent=0.115, tangentMode=RCTM_User, time=1, value=1}
         ]}}})
     Y=&FloatFromCurve(
      FloatCurve=DataInterface() {propertyValues {Curve {keys [{arriveTangent=1.5, interpMode=RCIM_Cubic, time=0, value=1.25}, {tangentMode=RCTM_User, time=1, value=1.5}]}}}
      Scale Curve=&FN_FrontEndFloatModifier(Gameplay=1, Front End=0.7)
     )
    ))
   &ScaleColor(
    Scale Alpha=&FloatFromCurve(
     FloatCurve=DataInterface() {propertyValues {Curve {keys [{interpMode=RCIM_Cubic, tangentMode=RCTM_User, time=0, value=1}, {interpMode=RCIM_Cubic, tangentMode=RCTM_User, time=1, value=0}]}}}
     Scale Curve=&FloatFromCurve(
      FloatCurve=DataInterface() {propertyValues {Curve {keys [
          {interpMode=RCIM_Cubic, tangentMode=RCTM_User, time=0, value=0}
          {interpMode=RCIM_Cubic, tangentMode=RCTM_User, time=0.5, value=1}
          {interpMode=RCIM_Cubic, tangentMode=RCTM_User, time=0.6, value=1}
          {tangentMode=RCTM_User, time=1, value=0}
         ]}}}
      Scale Curve=&FN_FrontEndFloatModifier(Gameplay=1, Front End=0.7)
     )
    )
    Scale RGB=&MakeVectorFromLinearColorRGB(Color=<rgba 3 3 3 1>)
    ScaleA=true
    ScaleRGB=true
   )
   /SetVariables(Particles.SpriteSize=&Multiply_Vector2D_ByFloat(
     Float=&FloatFromCurve(FloatCurve=DataInterface() {propertyValues {Curve {keys [{arriveTangent=4.3493, interpMode=RCIM_Cubic, time=0, value=0.8}, <at 1 1 RCIM_Cubic>]}}})
     Vector2D=$Particles.SpriteSize
    )) {isSetParameters=true}
   &ScaleColor(
    Scale Alpha=&FloatFromCurve(
     FloatCurve=DataInterface() {propertyValues {Curve {keys [
         <at 0 0 RCIM_Cubic>
         {interpMode=RCIM_Cubic, tangentMode=RCTM_User, time=0.05, value=1}
         {interpMode=RCIM_Cubic, tangentMode=RCTM_User, time=0.75, value=1}
         <at 0.8 0>
        ]}}}
     CurveIndex=$Emitter.NormalizedLoopAge
    )
    ScaleA=true
    ScaleRGB=false
   ) {module=ScaleColor001}
  ]
  rendererClasses [&NiagaraSpriteRendererProperties]
  renderers [{rendererClass=&NiagaraSpriteRendererProperties}]
 }
}
summary {
 systemName=NS_Emote_Sunlight_02
 emitters [
  {emitterName=Headlights, rendererClasses [&NiagaraSpriteRendererProperties, &NiagaraLightRendererProperties]}
  {emitterName=LightRays, rendererClasses [&NiagaraSpriteRendererProperties]}
 ]
 userVariables [
  User.bIsFrontEnd {type=NiagaraBool}
  User.bIsFrontEndPreview {type=NiagaraBool}
 ]
}
